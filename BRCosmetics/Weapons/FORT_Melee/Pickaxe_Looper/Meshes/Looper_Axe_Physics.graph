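# Physics asset for the Looper pickaxe, a two-handed melee cosmetic, defining simple collision and joints for its dangling cloth strips.
# Three small capsule bodies cover the handle bone and the main and little cloth strip bones, with the strip capsules offset and rotated along their bones.
# Each strip joins directly to the handle through a constraint with limited swing and tighter twist, so the two move independently of each other.
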PhysicsAsset /BRCosmetics/Weapons/FORT_Melee/Pickaxe_Looper/Meshes/Looper_Axe_Physics
bodies {
 Melee_TwoHanded_Handle [{shapeType=Capsule, length=1, radius=1, shapeName=Melee_TwoHanded_Handle_capsule}]
 dyn_cloth_strip_little [{shapeType=Capsule, length=1, radius=1, shapeName=dyn_cloth_strip_little_capsule, center=<1.6116 0 0>, rotation=<rot -90 0 0>}]
 dyn_cloth_strip_main [{shapeType=Capsule, length=1, radius=1, shapeName=dyn_cloth_strip_main_capsule, center=<2.805 0 0>, rotation=<rot -90 0 0>}]
}
constraints [
 {bone1Name=dyn_cloth_strip_little, bone2Name=Melee_TwoHanded_Handle, swing1LimitDegrees=45, swing2LimitDegrees=45, twistLimitDegrees=10}
 {bone1Name=dyn_cloth_strip_main, bone2Name=Melee_TwoHanded_Handle, swing1LimitDegrees=45, swing2LimitDegrees=45, twistLimitDegrees=10}
]
