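# Physics asset for the upper part of the Looper pickaxe, a cosmetic two-handed melee weapon.
# It defines small capsule bodies on the handle bone and two dynamic arm bones.
# Constraints chain the second arm to the first and the first to the handle, with tight twist and swing limits that leave slightly more freedom on the second swing axis for subtle, mostly single-axis wobble.
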PhysicsAsset /BRCosmetics/Weapons/FORT_Melee/Pickaxe_Looper/Meshes/Looper_Axe_Physics_UpperPart
bodies {
 Melee_TwoHanded_Handle [{shapeType=Capsule, length=1, radius=1, shapeName=Melee_TwoHanded_Handle_capsule}]
 dyn_arm_01 [{shapeType=Capsule, length=2, radius=3, shapeName=dyn_arm_01_capsule, center=<2 0 0>, rotation=<rot -90 0 0>}]
 dyn_arm_02 [{shapeType=Capsule, length=2, radius=3, shapeName=dyn_arm_02_capsule, center=<6 0 0>, rotation=<rot -90 0 0>}]
}
constraints [
 {bone1Name=dyn_arm_02, bone2Name=dyn_arm_01, swing1LimitDegrees=1, swing2LimitDegrees=7, twistLimitDegrees=1}
 {bone1Name=dyn_arm_01, bone2Name=Melee_TwoHanded_Handle, swing1LimitDegrees=1, swing2LimitDegrees=8, twistLimitDegrees=1}
]
